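; Weakpoint component for the ToyFox giant pawn that tracks an Inactive, Active or Destroyed state, gated by a hotfix flag.
; The server changes state and destroys an active weakpoint once damage from NotifyDamage exceeds a threshold.
; The state's RepNotify updates query collision, persistent Niagara and indexed mesh material parameters, while a multicast plays one-off bursts.
; The server also notifies the game state component on activation and submits lifetime analytics at EndPlay.
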(event EventBeginPlay
  ;[0 0.15 0.007618 0.5] BeginPlay: Set the weakpoint to inactive if it is enabled as a weakpoint.
  (|CallParentFunction)
  (Networking|SwitchHasAuthority
    (:Authority
      (Class|FortAsyncActionCurrentPlaylistReady|PlaylistDataReadyAsync
        (:OnReady
          (if (ScalableFloat|GetValueasBool (Variables|Hotfix|GetHotfix_WeakpointEnable))
            (State|SetWeakpointState "Inactive") ; Only inactive weakpoints can be come active.
          ))))))

(event EventEndPlay
  (params (EndPlayReason :type EEndPlayReason%20Enum :default "Destroyed"))
  ;[0.08333 0.04045 0 0.5] EndPlay: Clean up.
  (Utilities|FlowControl|Sequence
    (:then_0
      (Networking|SwitchHasAuthority
        (:Authority
          (Analytics|SubmitAnalyticstoGameStateComponent))
        (:Remote
          (FX|SetWeakpointNiagaraEffectEnabled false))))
    (:then_1
      (|CallParentFunction EndPlayReason))))

(event Custom|UpdateWeakpoint
  (bind _returnvalue (Fortnite|Sync|Time|GetServerWorldTimeSeconds))
  ;[0.15 0 0.1052 0.5] UpdateWeakpoint: The weakpoint state has been updated. Update the collision. Update the visuals based on the state.
  ; This event runs on both server and client.
  (Collision|SetCollisionEnabled (select (Utilities|Enum|Equal(Enum) (Variables|Transient|GetWeakpointState) "Active") "QueryOnly" "NoCollision"))
  (switch Utilities|FlowControl|Switch|SwitchonEFortWeakPointState (Variables|Transient|GetWeakpointState)
    (:Active
      (FX|SetWeakpointNiagaraEffectEnabled true)
      (FX|SetWeakpointActiveMaterialEnabled true)
      (Networking|SwitchHasAuthority
        (:Authority
          (Variables|Transient|Analytics|SetWeakpointActiveServerTime _returnvalue)
          (bind _gamestatecomponent (Utilities|GetGameStateComponent)
            (:Is%20Valid
              (Analytics|Notify|NotifyGiantWeakpointCreated _gamestatecomponent))))))
    (:Destroyed
      (FX|SetWeakpointNiagaraEffectEnabled false)
      (FX|SetWeakpointActiveMaterialEnabled false)
      (Networking|SwitchHasAuthority
        (:Authority
          (Variables|Transient|Analytics|SetWeakpointDestroyedServerTime _returnvalue))
        (:Remote
          (GameplayCue|ExecuteGameplayCue%28Burst%29Local (Components|GetOwner) (Variables|Default|GetGC_WeakpointDestroy) (Ability|GameplayCue|MakeGameplayCueParameters :NormalizedMagnitude 0.0 :RawMagnitude 0.0 :Location (Transformation|GetWorldLocation) :Normal "0,0,0" :SourceObject self :GameplayEffectLevel 1 :AbilityLevel 1 :bReplicateLocationWhenUsingMinimalRepProxy false)))))))

(event Custom|MulticastWeakpointStateUpdated
  (params (NewWeakpointState :type EFortWeakPointState%20Enum :default "Uninitialized"))
  ;[0.04669 0.002533 0.15 0.5] HandleWeakpointStateUpdatedBurst: This should be used to spawn burst effects, such as burst sounds, or burst particle effects for when the weakpoint state changes.
  ; It should not be used to make state changes. State changes such as materials changing color or persistent niagara effects should be handled in the UpdateWeakpoint event above.
  (Networking|CanExecuteCosmeticEvents
    (:True
      (Utilities|FlowControl|Delay 0.033)
      (switch Utilities|FlowControl|Switch|SwitchonEFortWeakPointState NewWeakpointState
        (:Active
          (Niagara|SpawnSystemAttached (Variables|FX|GetWeakpointActivatedBurstNiagaraSystem) self "None" "0, 0, 0" "0, 0, 0" "SnapToTarget" true true "AutoRelease")
          (Audio|SpawnSoundAttached "/ToyFoxGameplay/Assets/Sound/SC_ToyFox_SFX_Damage_Weakpoint_Expl.SC_ToyFox_SFX_Damage_Weakpoint_Expl" self))
        (:Destroyed
          (Niagara|SpawnSystemAttached (Variables|FX|GetWeakpointDestroyedBurstNiagaraSystem) self "None" "0, 0, 0" "0, 0, 0" "SnapToTarget" true true "AutoRelease"))))))

(fn GetMeshComponent
  (Utilities|FlowControl|Sequence
    (:then_0
      (bind _astoy_fox_giant_pawn_interface (Utilities|Casting|CastToToyFoxGiantPawnInterface (Components|GetOwner)))
      (bind _mesh (Class|ToyFoxGiantPawnInterface|GetSkeletalMeshBudgeted _astoy_fox_giant_pawn_interface))
      (return _mesh))
    (:then_1
      (return 0))))

(fn GetWeakpointDestroyBonusDamage
  ;[0.01022 0.06925 0.15 0.5] Returns the bonus damage that should be done when this weakpoint is destroyed.
  (return (Math|Float|Max(Float) (ScalableFloat|GetCachedScalableFloatValue (Variables|Hotfix|GetHotfix_WeakpointDestroyedBonusDamage)))))

(fn GetWeakpointDestroyDamagePending
  ;[0.01754 0.01697 0.15 0.5] Returns how much damage is left before the weakpoint is going to be destroyed.
  (return (Math|Float|Max(Float) (- (ScalableFloat|GetCachedScalableFloatValue (Variables|Hotfix|GetHotfix_WeakpointActiveDamageThreshold)) (Variables|Transient|GetWeakpointDamageTaken)))))

(fn GetWeakpointState
  (return (Variables|Transient|GetWeakpointState)))

(fn IsWeakpointActive
  ;[0 0.07487 0.15 0.5] Returns if the weakpoint is active and should be taking bonus damage.
  (return (Utilities|Enum|Equal(Enum) (State|GetWeakpointState) "Active")))

(fn Native_IsWeakpointActive
  (return (Utilities|Enum|Equal(Enum) (Variables|Transient|GetWeakpointState) "Active")))

(fn NotifyDamage
  (params (Damage :type Float%20%28double-precision%29))
  ;[0.003839 0 0.15 0.5] NotifyDamage: Track damage taken by the weakpoint when active. After threshold limit, destroy the weakpoint.
  ; This is called from the B_ToyFoxGiantPlayerPawn blueprint.
  (Networking|SwitchHasAuthority
    (:Authority
      (if (Utilities|Enum|Equal(Enum) (State|GetWeakpointState) "Active")
        (bind _output_get (Variables|Transient|SetWeakpointDamageTaken (+ (Variables|Transient|GetWeakpointDamageTaken) Damage)))
        (if (> _output_get (ScalableFloat|GetCachedScalableFloatValue (Variables|Hotfix|GetHotfix_WeakpointActiveDamageThreshold)))
          (State|SetWeakpointState "Destroyed"))))))

(fn OnRep_WeakpointState
  (CallFunction|UpdateWeakpoint))

(fn SetWeakpointActiveMaterialEnabled
  (params (bEnable :type Boolean))
  (bind _weakpointmaterialdataindex (Variables|Instance|GetWeakpointMaterialDataIndex))
  (bind _returnvalue (/ _weakpointmaterialdataindex 4))
  (Networking|CanExecuteCosmeticEvents
    (:True
      (bind _mesh (Helper|GetMeshComponent))
      (Utilities|IsValid _mesh
        (:Is%20Valid
          (Rendering|Material|SetScalarParameterValueonMaterials _mesh (Utilities|Select :Index _returnvalue :Option%200 "Weakpoint01Stage" :Option%201 "Weakpoint02Stage" :Option%202 "Weakpoint03Stage" :Option%203 "Weakpoint04Stage" :Option%204 "Weakpoint05Stage" :Option%205 "Weakpoint06Stage" :Option%206 "Weakpoint07Stage" :Option%207 "Weakpoint08Stage") (Math|Float|SelectFloat 1.0 2.0 bEnable))
          (Rendering|Material|SetScalarParameterValueonMaterials _mesh "Animation_Duration" (Math|Float|SelectFloat 1.8 3.2 bEnable)) ; Set Duration
          (Rendering|Material|SetScalarParameterValueonMaterials _mesh (Utilities|Select :Index _returnvalue :Option%200 "GameTimeFromBP01" :Option%201 "GameTimeFromBP02" :Option%202 "GameTimeFromBP03" :Option%203 "GameTimeFromBP04" :Option%204 "GameTimeFromBP05" :Option%205 "GameTimeFromBP06" :Option%206 "GameTimeFromBP07" :Option%207 "GameTimeFromBP08") (Utilities|Time|GetGameTimeinSeconds)) ; Passthough Time Individrually
        )))))

(fn SetWeakpointNiagaraEffectEnabled
  (params (bEnable :type Boolean))
  ;[0.15 0.02648 0.05439 0.5] SetWeakpoinNiagaraEffectEnabled: Enable or disable the weakpoint niagara effect.
  (Networking|CanExecuteCosmeticEvents
    (:True
      (if bEnable
        (Variables|Default|GetWeakpointNiagaraSystemComponent
          (:then)
          (:else
            (bind _returnvalue (Niagara|SpawnSystemAttached (Variables|FX|GetWeakpointNiagaraSystem) self "None" "0, 0, 0" "0, 0, 0" "SnapToTarget" true true "AutoRelease"))
            (Variables|Default|SetWeakpointNiagaraSystemComponent _returnvalue)))
        (else
          (bind _weakpointniagarasystemcomponent (Variables|Default|GetWeakpointNiagaraSystemComponent)
            (:then
              (Components|Activation|Deactivate _weakpointniagarasystemcomponent)
              (Variables|Default|SetWeakpointNiagaraSystemComponent 0))
            (:else)))))))

(fn SetWeakpointState
  (params (NewState :type EFortWeakPointState%20Enum :default "Uninitialized"))
  ;[0.15 0.000116 0.03184 0.5] SetWeakpointState: Updating weakpoint state to the desired value.
  (Networking|SwitchHasAuthority
    (:Authority
      (if (Utilities|Enum|NotEqual(Enum) (Variables|Transient|GetWeakpointState) NewState)
        (bind _output_get (Variables|Transient|SetWeakpointState NewState))
        (CallFunction|MulticastWeakpointStateUpdated _output_get)))))

(fn SubmitAnalyticsToGameStateComponent
  ;[0.02778 0 0.15 0.5] SubmitAnalyticsToGameStateComponent: Send the analytics information to the game state component.
  (Utilities|SwitchNetRole
    (:Server
      (bind _gamestatecomponent (Utilities|GetGameStateComponent)
        (:Is%20Valid
          (if (Utilities|Enum|Equal(Enum) (State|GetWeakpointState) "Destroyed")
            (Analytics|AddAnalyticsAttribute :self _gamestatecomponent :AttributeName (Utilities|String|Append "WeakpointLifetime : " (Variables|Instance|GetWeakpointName)) :IntValue -1 :floatValue (- (Variables|Transient|Analytics|GetWeakpointDestroyedServerTime) (Variables|Transient|Analytics|GetWeakpointActiveServerTime)) :BoolValue false)))))))
